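# Lit material for Niagara mesh particles in the Grapefruit backpack cosmetic VFX.
# Dynamic parameters animate atlas cells over time, choosing particle-supplied or mesh UVs to sample color and mask textures.
# Named reroutes organize base color with particle tint and optional LUT, axis-based edge darkening, and a masked world position offset that can collapse vertices toward particle centers.
# Particle U feeds both UV channels, and y_axis_x fills all Y_Axis components.
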Material /BRCosmetics/Accessories/FORT_Backpacks/Backpack_F_MED_Grapefruit/VFX/M_Grapefruit_Backpack
graph {
 properties {bTangentSpaceNormal=false, bUsedWithNiagaraMeshParticles=true, pixelDepthOffsetMode=PDOM_Legacy, usedShadingModels=Default Lit}
 nodes [
  n1 = NamedRerouteUsage()
  n2 = TextureSample() {texture=/Game/Effects/Fort_Effects/Effects/Characters/Generic/ResIn/SquareGradient}
  LinearInterpolate(A=Multiply(A=@n1) {constB=0.8}, B=@n1, Alpha=Saturate(None=@n2.RGB))
  LinearInterpolate(
   A=VertexNormalWS()
   B=Subtract(A=WorldPosition().XYZ, B=ParticlePositionWS())
   Alpha=ScalarParameter() {defaultValue=0.02, parameterName=SpherizeNormals}
  ) {constAlpha=0.02}
  n3 = ScalarParameter() {defaultValue=0.2, parameterName=BaseEmissive, sortPriority=100}
  n4 = Multiply(A=@n2.RGB, B=ScalarParameter() {defaultValue=0.2, parameterName=EdgeEmissive, sortPriority=101}) {constB=0.2}
  MaterialFunctionCall(Emissive=Add(A=Multiply(A=@n3, B=@n1), B=Multiply(A=@n1, B=@n4))) {materialFunction=&MF_FX_EmissiveAdjustments}
  Multiply() {constB=0.2}
  NamedRerouteDeclaration(None=Transform(None=Constant3Vector() {constant=<rgba 0 1 0 0>}) {transformSourceType=TRANSFORMSOURCE_Local}) {name=WorldSpaceYAxis, nodeColor=<rgba 1 0.9844 0 1>}
  n5 = DynamicParameter() {parameterIndex=1, defaultValue=<rgba 1 1 4 1>, paramNames [cell_count,cell_speed,roughness,NA]}
  n6 = Multiply(A=Time(), B=@n5.cell_speed)
  n7 = DynamicParameter() {defaultValue=<rgba 1 1 4 1>, paramNames [U, V, column_offset, row_offset]}
  n8 = Add(A=Fmod(A=Ceil(None=@n6), B=@n5.cell_count), B=Ceil(None=@n7.column_offset))
  n9 = ScalarParameter() {defaultValue=8, parameterName=column_count, sortPriority=1}
  n10 = Divide(B=@n9)
  n11 = MaterialFunctionCall(Float2=LinearInterpolate(A=AppendVector(A=@n7.U, B=@n7.U), B=TextureCoordinate(), Alpha=ScalarParameter() {parameterName=UseTexCoords})) {materialFunction=&BreakOutFloat2Components}
  n12 = Add(A=Floor(None=Divide(A=@n8, B=@n9)), B=Floor(None=@n7.row_offset))
  n13 = Divide(B=ScalarParameter() {defaultValue=8, parameterName=row_count, sortPriority=0})
  n14 = MaterialFunctionCall(
   X=Add(A=Multiply(A=@n8, B=@n10), B=Multiply(A=@n11.R, B=@n10))
   Y=Add(A=Multiply(A=@n12, B=@n13), B=Multiply(A=@n11.G, B=@n13) {constB=0.125}) {constB=0.875}
  ) {materialFunction=&MakeFloat2}
  n15 = TextureSample(UVs=@n14.Result) {texture=&/T_Grapefruit_Backpack_D}
  NamedRerouteDeclaration(None=@n15.A) {name=TextureAlpha, nodeColor=<rgba 0 0.8594 1 1>}
  n16 = DynamicParameter() {parameterIndex=2, defaultValue=<rgba 1 1 4 1>, paramNames [y_axis_x,y_axis_y,y_axis_z,use_color]}
  NamedRerouteDeclaration(None=@n16.use_color) {name=UseColor, nodeColor=<rgba 0.1953 1 0 1>}
  n17 = TextureSample(UVs=@n14.Result) {samplerType=SAMPLERTYPE_LinearColor, texture=&/T_Grapefruit_Backpack_M}
  NamedRerouteDeclaration(None=@n17.R) {name=OffsetMask, nodeColor=<rgba 1 0.2578 0 1>}
  NamedRerouteDeclaration(None=@n15.RGB) {name=TextureColor, nodeColor=<rgba 1 0 0.5391 1>}
  n18 = Abs(None=DotProduct(A=NamedRerouteUsage(), B=VertexNormalWS()))
  n19 = ScalarParameter() {defaultValue=0.6, parameterName=EdgeDarkenMultAmount}
  NamedRerouteDeclaration(None=Saturate(None=Add(A=@n18, B=@n19) {constB=0.6})) {name=EdgeDarken, nodeColor=<rgba 1 0.375 0 1>}
  NamedRerouteDeclaration(None=@n5.roughness) {name=Roughness, nodeColor=<rgba 1 0.3281 0 1>}
  n20 = NamedRerouteUsage()
  n21 = Multiply(A=NamedRerouteUsage(), B=ScalarParameter() {parameterName=OffsetAmount})
  n22 = Step(Y=ComponentMask(None=NamedRerouteUsage()) {a=true}) {constX=0.5}
  n23 = Multiply(A=Subtract(A=WorldPosition().XYZ, B=ParticlePositionWS())) {constB=-1}
  NamedRerouteDeclaration(None=Add(A=Multiply(A=@n20, B=@n21), B=Multiply(A=@n22, B=@n23))) {name=WPO, nodeColor=<rgba 0 1 0.7656 1>}
  n24 = NamedRerouteUsage()
  n25 = LinearInterpolate(A=NamedRerouteUsage(), B=ParticleColor().RGB, Alpha=Multiply(A=@n24, B=NamedRerouteUsage()))
  n26 = LinearInterpolate(A=@n25, B=Multiply(A=@n25, B=NamedRerouteUsage()), Alpha=@n24)
  NamedRerouteDeclaration(None=LinearInterpolate(
    A=@n26
    B=MaterialFunctionCall!(In=@n26, LUT, LUT Multiplier) {materialFunction=/Game/Packages/Fortress_SharedMaterials/Anime_Master_Material/MF_ApplyLUT}.Result
    Alpha=ScalarParameter() {parameterName=ApplyLut}
   )) {name=BaseColor, nodeColor=<rgba 0 0.7891 1 1>}
  NamedRerouteDeclaration(None=OneMinus(None=Saturate(None=@n18))) {name=SideFace, nodeColor=<rgba 0 1 0.7656 1>}
  NamedRerouteDeclaration(None=MaterialFunctionCall(X=@n16.y_axis_x, Y=@n16.y_axis_x, Z=@n16.y_axis_x) {materialFunction=&MakeFloat3}.Result) {name=Y_Axis, nodeColor=<rgba 0.9688 0 1 1>}
  NamedRerouteUsage()
  NamedRerouteUsage()
  NamedRerouteUsage()
  NamedRerouteUsage()
  ObjectOrientation()
  PixelNormalWS()
  ScalarParameter() {parameterName=row_offset, sortPriority=2}
  ScalarParameter() {defaultValue=4, parameterName=cell_count, sliderMax=8, sliderMin=1, sortPriority=4}
  ScalarParameter() {parameterName=column_offset, sortPriority=3}
  ScalarParameter() {defaultValue=10, desc=Speed, parameterName=speed, sortPriority=6}
 ]
}
