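# Niagara system for the Grapefruit backpack cosmetic that draws a small pixel-grid display from particles.
# One burst spawns a square grid of particles that never expire, each given a UV and grid position by scratch dynamic inputs, rendered through sprite and mesh renderers.
# Every update, user glitch parameters perturb positions and UVs.
# Dynamic material parameters pass those UVs, cell animation and offset settings, colour toggle and owner Y axis to the material.
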NiagaraSystem /BRCosmetics/Accessories/FORT_Backpacks/Backpack_F_MED_Grapefruit/VFX/NS_Grapefruit_Backpack
dependencies {usedDataInterfaces [], usedDynamicInputs [], usedModules [&SystemState]}
emitters {Empty {
  emitterSpawnScript [/SetVariables(Emitter.PixelRadius=1.75, Emitter.PixelsPerRow=16) {isSetParameters=true}]
  emitterUpdateScript [
   &EmitterState(
    Inactive Response=::NewEnumerator1
    Life Cycle Mode=::NewEnumerator1
    Loop Behavior=::NewEnumerator1
    Loop Duration Mode=::NewEnumerator1
    UseLoopDelay=false
   )
   &SpawnBurst_Instantaneous(
    Spawn Count=/Niagara/DynamicInputs/Math/Min_Int(A:NiagaraInt32=256, B:NiagaraInt32=&Multiply_Int(A=$Emitter.PixelsPerRow, B=$Emitter.PixelsPerRow))
    Use Spawn Probability=false
   )
  ]
  particleSpawnScript [
   &FN_SetSpawnGroup(
    Inclusivity=true
    Normalize Spawn Group=true
    Number of Spawn Groups:NiagaraInt32=$Emitter.PixelsPerRow
    Particles Per Group:NiagaraInt32=$Emitter.PixelsPerRow
   )
   &InitializeParticle(
    Color=$User.color
    Lifetime=1.5
    Mesh Scale Mode=::NewEnumerator3
    Mesh Uniform Scale=&Multiply_Float(A=$Emitter.PixelRadius, B=1)
    Sprite Size Mode=::NewEnumerator3
    Uniform Sprite Size=&Multiply_Float(A=$Emitter.PixelRadius, B=0.5)
    UsePositionOffset=false
    Write Material Random=true
   )
   /SetVariables(Particles.ParticleUV:Vector2f=&self:ScratchDynamicInput_01(
     PixelsPerRow:NiagaraFloat=&MakeFloatFromInt(Int=$Emitter.PixelsPerRow)
     SpawnGroup:NiagaraFloat=&MakeFloatFromInt(Int=$Particles.SpawnGroup)
     SpawnGroupSeed:NiagaraFloat=&MakeFloatFromInt(Int=$Particles.SpawnGroupSeed)
    )) {isSetParameters=true}
   /SetVariables(Particles.ParticlePositionGrid:Vector3f=&self:ParticleCoordsVector(
     PixelRadius:NiagaraFloat=$Emitter.PixelRadius
     PixelsPerRow:NiagaraFloat=&MakeFloatFromInt(Int=$Emitter.PixelsPerRow)
     Spacing:NiagaraFloat=$Emitter.ParticleSpacing
     SpawnGroup:NiagaraFloat=&MakeFloatFromInt(Int=$Particles.SpawnGroup)
     SpawnGroupSeed:NiagaraInt32=$Particles.SpawnGroupSeed
    )) {isSetParameters=true}
   /SetVariables(Particles.ParticleOffset:NiagaraFloat=&Lerp_Float(A=0.25, B=-0.25, Alpha=/Niagara/DynamicInputs/TypeConversions/VectorLengthSafe(Vector:Vector3f=$Particles.ParticlePositionGrid))) {isSetParameters=true}
   /SetVariables(Particles.RandomPerParticle:NiagaraFloat=&RandomRangeFloat(Minimum=0, Maximum=1)) {isSetParameters=true}
   &SphereLocation(
    MaskSpawnGroup=false
    Non Uniform Scale=<0.2 1 1>
    Override Seed=false
    Sphere Radius=150
    Surface Only=false
    UseNonUniformScale=true
    UseOffset=false
    UseSphereOrientation=false
   ) {enabled=false}
   &DynamicMaterialParameters(
    Index 0 Param 1=&MakeFloatFromVector2d(Vector 2D=$Particles.ParticleUV)
    Index 0 Param 2=&MakeFloatFromVector2d(Vector 2D=$Particles.ParticleUV, Channel=::NewEnumerator1)
    Index 1 Param 3=$Particles.RandomPerParticle
    Module.Param0WriteEnabled=true
    Module.Param0WriteEnabled1=false
    Module.Param1WriteEnabled=true
    Module.Param1WriteEnabled1=false
    Module.Param2WriteEnabled=false
    Module.Param2WriteEnabled1=true
    Module.Param3WriteEnabled=false
    Module.Param3WriteEnabled1=false
    Write Parameter Index 1=true
   )
  ]
  particleUpdateScript [
   &ParticleState(Kill Particles When Lifetime Has Elapsed=false)
   /SetVariables(Particles.ParticlePositionModified:Vector3f=&self:ScratchDynamicInput_04(
     GlitchAmount:NiagaraFloat=$User.glitch_amount
     GlitchSpeed:NiagaraFloat=$User.glitch_speed
     ParticleCoordinates:Vector3f=$Particles.ParticlePositionGrid
     ParticleIndex:NiagaraInt32=$Particles.UniqueID
     RandomPerParticle:NiagaraFloat=$Particles.RandomPerParticle
     YCoord:NiagaraFloat=&MakeFloatFromVector2d(Vector 2D=$Particles.ParticleUV, Channel=::NewEnumerator1)
    )) {isSetParameters=true}
   /SetVariables(Particles.ParticleUVModified:Vector2f=&self:ScratchDynamicInput_05(
     GlitchAmount:NiagaraFloat=$User.glitch_amount
     GlitchOffset:NiagaraInt32=$User.glitch_offset
     GlitchSpeed:NiagaraFloat=$User.glitch_speed
     ParticleIndex:NiagaraInt32=$Particles.UniqueID
     ParticleUV:Vector2f=$Particles.ParticleUV
     PixelsPerrow:NiagaraFloat=&MakeFloatFromInt(Int=$Emitter.PixelsPerRow)
     RandomPerParticle:NiagaraFloat=$Particles.RandomPerParticle
    )) {isSetParameters=true}
   /SetVariables(Particles.Position:Vector3f=$Particles.ParticlePositionModified) {isSetParameters=true}
   &DynamicMaterialParameters(
    Index 0 Param 1=&MakeFloatFromVector2d(Vector 2D=$Particles.ParticleUVModified)
    Index 0 Param 2=&MakeFloatFromVector2d(Vector 2D=$Particles.ParticleUVModified, Channel=::NewEnumerator1)
    Index 0 Param 3=$User.column_offset
    Index 0 Param 4=$User.row_offset
    Index 1 Param 1=$User.cell_count
    Index 1 Param 2=$User.cell_speed
    Index 1 Param 3=&Lerp_Float(A=0.5, B=0.6, Alpha=$Particles.RandomPerParticle)
    Index 2 Param 1=&MakeFloatFromVector(Vector=$Engine.Owner.SystemYAxis, Channel=::NewEnumerator0)
    Index 2 Param 2=&MakeFloatFromVector(Vector=$Engine.Owner.SystemYAxis, Channel=::NewEnumerator1)
    Index 2 Param 3=&MakeFloatFromVector(Vector=$Engine.Owner.SystemYAxis, Channel=::NewEnumerator2)
    Index 2 Param 4=$User.use_color
    Module.Param0WriteEnabled=true
    Module.Param0WriteEnabled1=true
    Module.Param0WriteEnabled2=true
    Module.Param1WriteEnabled=true
    Module.Param1WriteEnabled1=true
    Module.Param1WriteEnabled2=true
    Module.Param2WriteEnabled=true
    Module.Param2WriteEnabled1=true
    Module.Param2WriteEnabled2=true
    Module.Param3WriteEnabled=true
    Module.Param3WriteEnabled1=false
    Module.Param3WriteEnabled2=true
    Write Parameter Index 1=true
    Write Parameter Index 2=true
   ) {module=DynamicMaterialParameters001}
  ]
  rendererClasses [&NiagaraSpriteRendererProperties, &NiagaraMeshRendererProperties]
  renderers [
   {rendererClass=&NiagaraSpriteRendererProperties}
   {rendererClass=&NiagaraMeshRendererProperties, rendererIndex=1}
  ]
 }}
summary {
 systemName=NS_Grapefruit_Backpack
 emitters [{emitterName=Empty, rendererClasses [&NiagaraSpriteRendererProperties, &NiagaraMeshRendererProperties]}]
 userVariables [
  User.cell_count {type=NiagaraFloat, defaultValue=4, description=""}
  User.cell_speed {type=NiagaraFloat, defaultValue=10, description=""}
  User.color {type=LinearColor, description="", defaultValue {struct=LinearColor, value=<rgba 1 1 1 1>}}
  User.column_offset {type=NiagaraFloat, defaultValue=4, description=""}
  User.EdgeEmissive {type=NiagaraFloat, defaultValue=0, description=""}
  User.glitch_amount {type=NiagaraFloat, defaultValue=0, description=""}
  User.glitch_offset {type=NiagaraInt32, defaultValue=2}
  User.glitch_speed {type=NiagaraFloat, defaultValue=10, description=""}
  User.row_offset {type=NiagaraFloat, defaultValue=0, description=""}
  User.spherizeNormals {type=NiagaraFloat, defaultValue=0, description=""}
  User.use_color {type=NiagaraFloat, defaultValue=0, description=""}
 ]
}
